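# M_KeysText is an unlit translucent material for the MusicBlock item, drawing on the piano key texture from the BigPiano quest interactable.
# Texture coordinates are tiled down to show a fraction of the texture, shifted by the red and green channels of a UV Offset vector parameter to select the region.
# The sampled color is multiplied down to a dim level.
# A standalone constant and scalar parameter are also present, with no visible connections.
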Material /Game/Playgrounds/Items/MusicBlock/M_KeysText
graph {
 properties {blendMode=BLEND_Translucent, pixelDepthOffsetMode=PDOM_Legacy, shadingModel=MSM_Unlit, usedShadingModels=Unlit}
 nodes [
  Constant() {r=0.35}
  n1 = TextureCoordinate() {uTiling=0.2, vTiling=0.2}
  n2 = ComponentMask(None=VectorParameter() {parameterName=UV Offset}.RGB) {g=true, r=true}
  Multiply(
   A=Constant() {r=0.1}
   B=TextureSample(UVs=Add(A=@n1, B=@n2)) {texture=/Game/Athena/Items/QuestInteractables/BigPiano/Textures/T_piano_keys_01_D}.RGB
  )
  ScalarParameter()
 ]
}
