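# Master material for the Playgrounds music block: a metallic Default Lit surface with parameterized base color and a colored glow.
# T_MusicBlocks masks the ColorGlow tint into a dim idle emissive and a brighter lit one with fill and center glow.
# The Glow scalar blends between these two states inside a material attributes setup, which also feeds an MF_CreativeMoveTool function call.
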Material /Game/Playgrounds/Items/MusicBlock/M_MusicBlock_Master
graph {
 properties {bUseMaterialAttributes=true, pixelDepthOffsetMode=PDOM_Legacy, usedShadingModels=Default Lit}
 nodes [
  n1 = VectorParameter() {parameterName=ColorGlow, defaultValue=<rgba 1 0.3035 0.0104 1>}
  n2 = Multiply(B=@n1.RGB) {constA=0.02}
  n3 = TextureSample() {texture=&/T_MusicBlocks}
  n4 = Multiply(A=Multiply(A=@n1.RGB, B=@n3.RGB), B=Constant() {r=0.11})
  Add(A=@n2, B=@n4)
  n5 = Multiply(A=ScalarParameter() {defaultValue=0.2, parameterName=GlowONfillglow}, B=@n1.RGB) {constA=0.2}
  n6 = Multiply(A=Multiply(A=@n1.RGB, B=@n3.RGB), B=ScalarParameter() {defaultValue=10, parameterName=GlowONcenterglow})
  n7 = SetMaterialAttributes!(
   MaterialAttributes
   Base Color=VectorParameter() {parameterName=ColorBase, defaultValue=<rgba 0.0262 0.0063 0.2083 0>}.RGB
   Metallic=ScalarParameter() {defaultValue=0.9238, parameterName=Metallic}
   Specular=ScalarParameter() {parameterName=Spec}
   Roughness=ScalarParameter() {defaultValue=0.2, parameterName=Roughness}
   Emissive Color=LinearInterpolate(A=LinearInterpolate(A=@n2, B=@n4, Alpha=@n3.RGB), B=Add(A=@n5, B=@n6), Alpha=ScalarParameter() {parameterName=Glow})
  ) {attributeSetTypes [
    #1
    #2
    #3
    #4
    #5
   ]}
  MaterialFunctionCall(MA In=@n7) {materialFunction=&MF_CreativeMoveTool}
  Multiply(A=@n3.RGB) {constB=1.5}
 ]
}
